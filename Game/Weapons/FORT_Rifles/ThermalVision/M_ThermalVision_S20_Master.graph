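# Post-process master material for thermal-vision rifle scopes, recolouring the scene as a heat view within a masked area.
# C++-updated scope vectors feed a material layer that returns screen UVs, mask and colour multiplier.
# Custom stencil and depth build hot/cold masks, with options for cloaked players and vehicles.
# Curve-atlas rows colour heat gradients from sun/view-lit normals (deferred) or scene luminance (forward/mobile); static switches select separated targets or tinted active screen.
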Material /Game/Weapons/FORT_Rifles/ThermalVision/M_ThermalVision_S20_Master
graph {
 properties {
  blendablePriority=-1
  materialDomain=MD_PostProcess
  pixelDepthOffsetMode=PDOM_Legacy
  stencilCompare=MSC_Greater
  usedShadingModels=Default Lit
 }
 nodes [
  n1 = NamedRerouteUsage()
  n2 = SceneTexture(UVs=@n1) {sceneTextureId=PPI_PostProcessInput0}
  Divide(A=Add(A=ComponentMask(None=@n2.Color) {b=true, g=true, r=true}))
  n3 = VectorParameter() {desc="This value is updated in C++\r\nWeapon Scope Attachment", parameterName=ScopeTargetPointWorldSpace, defaultValue=<rgba 0 1 0 0>}
  n4 = VectorParameter() {desc="This value is updated in C++\r\nWeapon Scope Attachment", parameterName=ScopeUpVectorWorldSpace, defaultValue=<rgba 0 0 1 0>}
  n5 = SetMaterialAttributes!(
   MaterialAttributes
   Base Color=@n3.RGB
   Emissive Color=VectorParameter() {desc="This value is updated in C++\r\nWeapon Scope Attachment", parameterName=ScopeEyePointWorldSpace}.RGB
   Subsurface Color=@n4.RGB
  ) {
   desc="Passing dynamic scopes\r\nParams to layer"
   attributeSetTypes [#1, #2, #3]
  }
  n6 = MaterialAttributeLayers!(None=@n5) {defaultLayers {
    blends []
    editorOnly {
     deletedParentLayerGuids []
     layerGuids [#4]
     layerLinkStates [NotFromParent]
     layerNames [Background]
     layerStates [true]
     restrictToBlendRelatives []
     restrictToLayerRelatives [false]
    }
    layers [/Game/Weapons/FORT_Rifles/ThermalVision/ThermalAnimation/ML_Thermal_Init]
   }}
  n7 = GetMaterialAttributes(None=@n6) {
   desc="UV 0: Screen Space UV\r\nUV 1 (R): Post Process Mask\r\nUV 1 (G): Toggle Thermal Scale\r\nEmissive: Color Multiplier"
   attributeGetTypes [#2, #5, #6]
  }
  NamedRerouteDeclaration(None=ComponentMask(None=@n7.Customized UV 0) {g=true, r=true}) {name=UV Stuff, nodeColor=<rgba 1 0 0.1875 1>}
  n8 = NamedRerouteUsage()
  n9 = NamedRerouteUsage()
  n10 = NamedRerouteUsage()
  n11 = Constant() {r=1}
  n12 = ScalarParameter() {defaultValue=1, parameterName=Vehicle Tempratures}
  n13 = CurveAtlasRowParameter(CurveTime=Multiply(A=ShadingPathSwitch(Default=@n9, Deferred=@n9, Forward=@n8, Mobile=@n8), B=If(A=@n10, "A > B"=@n11, "A == B"=@n12, "A < B"=@n11) {constB=20})) {
   atlas=/Game/Weapons/FORT_Rifles/ThermalVision/ThermalCurves/Atlas_ThermalVisions
   curve=/Game/Weapons/FORT_Rifles/ThermalVision/ThermalCurves/Curve_TG_Green
   defaultValue=1
   parameterName=Thermal Colors
   sortPriority=0
  }
  n14 = LinearInterpolate(
   A=VectorParameter() {parameterName=Heat Tint 2, defaultValue=<rgba 0.5 0.5 0.5 1>}.RGB
   B=VectorParameter() {parameterName=Heat Tint 1, defaultValue=<rgba 1 1 1 1>}.RGB
   Alpha=@n13.R
  )
  n15 = Multiply(
   A=Power(Base=@n13, Exp=ScalarParameter() {defaultValue=1, parameterName=BrightnessPower})
   B=ScalarParameter() {defaultValue=1, parameterName=BrightnessHeat}
  )
  n16 = NamedRerouteUsage()
  n17 = ScalarParameter() {defaultValue=1, parameterName=PlayerMultiplyAmount, sliderMax=5}
  NamedRerouteDeclaration(None=Add(A=Multiply(A=@n14, B=@n15), B=Multiply(A=@n16, B=@n17))) {name=Hot Stuff, nodeColor=<rgba 1 0.0075 0 1>}
  NamedRerouteDeclaration(None=ComponentMask(None=@n7.Customized UV 1) {g=true}) {name=Thermal Scale, nodeColor=<rgba 0 1 0.9063 1>}
  NamedRerouteDeclaration(None=Multiply(A=ComponentMask(None=@n7.Customized UV 1) {r=true}, B=NamedRerouteUsage())) {name=Mask Stuff, nodeColor=<rgba 0.1719 0 1 1>}
  NamedRerouteDeclaration(None=@n7.Emissive Color) {name=Emissive Stuff, nodeColor=<rgba 1 0.5525 0 1>}
  n18 = NamedRerouteUsage()
  n19 = Multiply(
   A=Desaturation(None=SceneTexture(UVs=@n18) {sceneTextureId=PPI_PostProcessInput1}.Color, Fraction=Constant() {r=1})
   B=ScalarParameter() {defaultValue=2.5, parameterName=Separate Translucency Mask Multiplier}
  )
  n20 = Constant()
  NamedRerouteDeclaration(None=ShadingPathSwitch(Default=@n19, Deferred=@n19, Forward=@n20, Mobile=@n20)) {name=Separate Translucency Mask, nodeColor=<rgba 0 1 0.1094 1>}
  n21 = Desaturation(None=NamedRerouteUsage(), Fraction=ScalarParameter() {defaultValue=0.2, group=Active Screen, parameterName=Pretint Desaturation})
  n22 = ScalarParameter() {defaultValue=1.1, group=Active Screen, parameterName=Pretint Multiplier, sortPriority=50}
  n23 = VectorParameter() {group=Active Screen, parameterName=Active Screen Tint, sortPriority=100, defaultValue=<rgba 0.7729 1 0.6756 1>}
  NamedRerouteDeclaration(None=Multiply(A=Multiply(A=@n21, B=@n22), B=@n23.RGB)) {name=Tinted Scene, nodeColor=<rgba 1 0 0.1875 1>}
  n24 = LinearInterpolate(A=TextureCoordinate(), B=NamedRerouteUsage(), Alpha=NamedRerouteUsage())
  NamedRerouteDeclaration(None=ComponentMask(None=SceneTexture(UVs=@n24) {sceneTextureId=PPI_PostProcessInput0}.Color) {b=true, g=true, r=true}) {name=Original Scene, nodeColor=<rgba 1 0 0.4165 1>}
  n25 = Desaturation(None=NamedRerouteUsage(), Fraction=ScalarParameter() {defaultValue=1, parameterName=OriginalSceneDesaturationAmount})
  n26 = Multiply(
   A=Power(
    Base=NamedRerouteDeclaration(None=@n25) {name=Original Scene Desaturated, nodeColor=<rgba 1 0 0.4165 1>}
    Exp=ScalarParameter() {defaultValue=0.45, parameterName=HeatGradientPower}
   )
   B=ScalarParameter() {defaultValue=2.6157, parameterName=HeatGradientMult}
  )
  NamedRerouteDeclaration(None=DotProduct(A=Saturate(None=@n26), B=Constant() {r=0.3333})) {name=Heat Gradient Forward, nodeColor=<rgba 1 0.1598 0.1042 1>}
  n27 = NamedRerouteUsage()
  n28 = CameraVectorWS()
  n29 = CollectionParameter() {
   collection=&FortniteMaterialParameters
   parameterName=SunlightDirectionalVector
  }
  n30 = ComponentMask(None=@n29) {b=true, g=true, r=true}
  n31 = ScalarParameter() {defaultValue=0.5, parameterName=Sun Direction Gradient}
  n32 = DotProduct(
   A=ComponentMask(None=SceneTexture(UVs=@n27) {sceneTextureId=PPI_WorldNormal}.Color) {b=true, g=true, r=true}
   B=Normalize(VectorInput=LinearInterpolate(A=@n28, B=@n30, Alpha=@n31) {constAlpha=0.3})
  )
  n33 = ScalarParameter() {defaultValue=2.2787, parameterName=FresnelGradient}
  NamedRerouteDeclaration(None=Multiply(A=Power(Base=@n32, Exp=@n33), B=ScalarParameter() {defaultValue=1.3, parameterName=FresnelMult}) {constB=0.1}) {name=Heat Gradient Deferred, nodeColor=<rgba 1 0.1598 0.1042 1>}
  n34 = NamedRerouteUsage()
  n35 = ScalarParameter() {defaultValue=0.7, parameterName=DesatSceneColor}
  NamedRerouteDeclaration(None=Multiply(A=VectorParameter() {parameterName=Background Tint, defaultValue=<rgba 0.0245 0.05 0.016 1>}.RGB, B=Desaturation(None=@n34, Fraction=@n35))) {name=Cold Stuff, nodeColor=<rgba 0 0.4934 1 1>}
  n36 = NamedRerouteUsage()
  n37 = NamedRerouteDeclaration(None=ComponentMask(None=SceneTexture(UVs=@n36) {sceneTextureId=PPI_CustomStencil}.Color) {r=true}) {name=Custom Stencil, nodeColor=<rgba 1 0 0.0625 1>}
  n38 = Constant() {r=1}
  n39 = If(A=@n37, B=Constant() {r=2}, "A > B"=@n38, "A == B"=Constant(), "A < B"=@n38) {desc=buildings highlight}
  n40 = Constant() {r=1}
  n41 = If(A=@n37, B=Constant() {r=123}, "A > B"=@n40, "A == B"=Constant(), "A < B"=@n40) {desc=Damage numbers highlight}
  n42 = Abs(None=Subtract(A=@n37) {constB=27})
  n43 = Constant() {r=1}
  n44 = Constant()
  n45 = Multiply(
   A=SmoothStep(Value=SceneDepth(UVs=@n36).R) {constMax=5000000, constMin=10000000, desc=Sky mask}
   B=Multiply(A=Multiply(A=@n39, B=@n41), B=If(A=@n42, "A > B"=@n43, "A == B"=@n44, "A < B"=@n44) {constB=1.01, desc="BelongTreaty FX (27-28)"})
  )
  n46 = Constant()
  n47 = Max(B=If(A=NamedRerouteUsage(), "A > B"=@n46, "A == B"=Constant() {r=1}, "A < B"=@n46) {constB=24})
  n48 = ScalarParameter() {parameterName=Reveal Cloaking Players}
  n49 = NamedRerouteUsage()
  n50 = SceneTexture(UVs=@n49) {sceneTextureId=PPI_CustomDepth}
  n51 = Saturate(None=Subtract(A=ComponentMask(None=@n50.Color) {r=true}, B=SceneDepth(UVs=@n49).R))
  n52 = NamedRerouteUsage()
  n53 = Constant() {r=1}
  n54 = Constant()
  n55 = Multiply(
   A=Max(A=@n45, B=LinearInterpolate(A=@n45, B=@n47, Alpha=@n48))
   B=Multiply(A=OneMinus(None=@n51), B=If(A=@n52, "A > B"=@n53, "A == B"=@n54, "A < B"=@n53) {constB=22})
  )
  n56 = Constant() {r=1}
  n57 = If(A=NamedRerouteUsage(), "A > B"=@n56, "A == B"=Constant(), "A < B"=@n56) {constB=20}
  NamedRerouteDeclaration(None=LinearInterpolate(A=@n55, B=Multiply(A=@n55, B=@n57), Alpha=ScalarParameter() {parameterName=Remove Vehicles})) {name=Hot/Cold Mask, nodeColor=<rgba 0.957 0.1708 1 1>}
  n58 = NamedRerouteUsage()
  n59 = NamedRerouteUsage()
  n60 = Multiply(A=ShadingPathSwitch(Default=@n58, Deferred=@n58, Forward=@n59, Mobile=@n59)) {constB=0.5}
  n61 = LinearInterpolate(Alpha=NamedRerouteUsage()) {constB=0.5}
  n62 = CurveAtlasRowParameter(CurveTime=DotProduct(A=Add(A=@n60, B=@n61), B=Constant() {r=0.3333})) {
   atlas=&Atlas_ThermalVisions
   curve=&Curve_TG_Green
   defaultValue=1
   parameterName=Thermal Colors
   sortPriority=0
  }
  n63 = NamedRerouteUsage()
  n64 = ScalarParameter() {parameterName=Add Desaturated Original Scene}
  NamedRerouteDeclaration(None=Add(A=@n62, B=LinearInterpolate(B=@n63, Alpha=@n64))) {name=All Stuff, nodeColor=<rgba 0 1 0.6016 1>}
  n65 = ComponentMask(None=SceneTexture(UVs=@n1) {sceneTextureId=PPI_ShadingModelID}.Color) {r=true}
  n66 = ScalarParameter() {defaultValue=5, parameterName=Shading Model}
  n67 = Constant() {r=1}
  Power(
   Base=ComponentMask(None=Desaturation(None=@n2.Color)) {r=true}
   Exp=LinearInterpolate(Alpha=SphereMask(A=@n65, B=@n66, Radius=@n67, Hardness=@n67)) {constA=1, constB=0.3}
  )
  n68 = LinearInterpolate(A=NamedRerouteUsage(), B=NamedRerouteUsage(), Alpha=NamedRerouteUsage())
  n69 = NamedRerouteUsage()
  n70 = StaticSwitchParameter(True=LinearInterpolate(A=NamedRerouteUsage(), B=NamedRerouteUsage(), Alpha=NamedRerouteUsage()), False=NamedRerouteUsage()) {defaultValue=true, parameterName=Seperate Targets}
  n71 = Power(Base=Multiply(A=NamedRerouteDeclaration(None=@n70) {name=Thermal Combined, nodeColor=<rgba 1 0.3047 0 1>}, B=NamedRerouteUsage())) {constExponent=1.2}
  n72 = ScalarParameter() {defaultValue=1.3, parameterName=Brightness}
  n73 = MaterialFunctionCall!(ID=Constant() {r=123}) {materialFunction=/Game/Packages/Fortress_SharedMaterials/Stencil_ID_Definitions/StencilMaskCompare}
  n74 = Constant()
  n75 = If(
   A=SceneDepth(UVs=TextureCoordinate()).R
   B=ScalarParameter() {defaultValue=17, parameterName=Scope Mesh Depth}
   "A > B"=Constant() {r=1}
   "A == B"=@n74
   "A < B"=@n74
  ) {constB=10}
  n76 = Multiply(
   A=NamedRerouteUsage()
   B=Multiply(A=OneMinus(None=@n73.Result), B=NamedRerouteDeclaration(None=@n75) {name=Scope Mask, nodeColor=<rgba 0 1 0.1562 1>})
  )
  StaticSwitchParameter(
   True=LinearInterpolate(A=NamedRerouteUsage(), B=Multiply(A=@n68, B=@n69), Alpha=NamedRerouteUsage())
   False=LinearInterpolate(
    A=NamedRerouteUsage()
    B=Multiply(A=@n71, B=@n72)
    Alpha=NamedRerouteDeclaration(None=@n76) {name=Masks Combined, nodeColor=<rgba 1 0.9844 0 1>}
   )
  ) {group=Active Screen, parameterName=UseActiveScreen, sortPriority=5}
 ]
}
