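# Post-process material that applies the weapon scope's thermal vision look to separate translucency, blending at Translucency After DOF.
# C++-updated scope vectors and thermal toggles pass through material attributes into the ML_Thermal_Init layer, which returns screen UVs, mask, scale and color multiplier.
# These drive resampling and tinting of the translucency input, with stencil compares limiting the affected pixels.
# The UseActiveScreen static switch selects between two tint blends.
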Material /Game/Weapons/FORT_Rifles/ThermalVision/M_ThermalVision_S29_SeparateTranslucency
graph {
 properties {
  blendableLocation=BL_TranslucencyAfterDOF
  blendablePriority=-1
  materialDomain=MD_PostProcess
  pixelDepthOffsetMode=PDOM_Legacy
  stencilCompare=MSC_Equal
  stencilRefValue=60
  usedShadingModels=Default Lit
 }
 nodes [
  Constant()
  n1 = VectorParameter() {desc="This value is updated in C++\r\nWeapon Scope Attachment", parameterName=ScopeTargetPointWorldSpace, defaultValue=<rgba 0 1 0 0>}
  n2 = VectorParameter() {desc="This value is updated in C++\r\nWeapon Scope Attachment", parameterName=ScopeUpVectorWorldSpace, defaultValue=<rgba 0 0 1 0>}
  n3 = SetMaterialAttributes!(
   MaterialAttributes
   Base Color=@n1.RGB
   Emissive Color=VectorParameter() {desc="This value is updated in C++\r\nWeapon Scope Attachment", parameterName=ScopeEyePointWorldSpace}.RGB
   Subsurface Color=@n2.RGB
   Opacity=ScalarParameter() {parameterName=ToggleThermalMask}
   Opacity Mask=ScalarParameter() {defaultValue=1, parameterName=ToggleThermalDirection}
  ) {
   desc="Passing dynamic scopes\r\nParams to layer"
   attributeSetTypes [
    #1
    #2
    #3
    #4
    #5
   ]
  }
  n4 = MaterialAttributeLayers!(None=@n3) {defaultLayers {
    blends []
    editorOnly {
     deletedParentLayerGuids []
     layerGuids [#6]
     layerLinkStates [NotFromParent]
     layerNames [Background]
     layerStates [true]
     restrictToBlendRelatives []
     restrictToLayerRelatives [false]
    }
    layers [/Game/Weapons/FORT_Rifles/ThermalVision/ThermalAnimation/ML_Thermal_Init]
   }}
  n5 = GetMaterialAttributes(None=@n4) {
   desc="UV 0: Screen Space UV\r\nUV 1 (R): Post Process Mask\r\nUV 1 (G): Toggle Thermal Scale\r\nEmissive: Color Multiplier"
   attributeGetTypes [#2, #7, #8]
  }
  NamedRerouteDeclaration(None=ComponentMask(None=@n5.Customized UV 0) {g=true, r=true}) {name=UV Stuff, nodeColor=<rgba 1 0 0.1875 1>}
  NamedRerouteDeclaration(None=ComponentMask(None=@n5.Customized UV 1) {g=true}) {name=Thermal Scale, nodeColor=<rgba 0 1 0.9063 1>}
  n6 = MaterialFunctionCall!(ID=Constant() {r=123}) {materialFunction=/Game/Packages/Fortress_SharedMaterials/Stencil_ID_Definitions/StencilMaskCompare}
  NamedRerouteDeclaration(None=Multiply(A=ComponentMask(None=@n5.Customized UV 1) {r=true}, B=OneMinus(None=@n6.Result))) {name=Mask Stuff, nodeColor=<rgba 0.1719 0 1 1>}
  NamedRerouteDeclaration(None=@n5.Emissive Color) {name=Emissive Stuff, nodeColor=<rgba 1 0.5525 0 1>}
  n7 = Desaturation(None=NamedRerouteUsage(), Fraction=ScalarParameter() {defaultValue=0.2, group=Active Screen, parameterName=Pretint Desaturation})
  n8 = ScalarParameter() {defaultValue=1.1, group=Active Screen, parameterName=Pretint Multiplier, sortPriority=50}
  n9 = VectorParameter() {group=Active Screen, parameterName=Active Screen Tint, sortPriority=100, defaultValue=<rgba 0.7729 1 0.6756 1>}
  NamedRerouteDeclaration(None=Multiply(A=Multiply(A=@n7, B=@n8), B=@n9.RGB)) {name=Tinted Scene, nodeColor=<rgba 1 0 0.1875 1>}
  n10 = SceneTexture(UVs=LinearInterpolate(A=TextureCoordinate(), B=NamedRerouteUsage(), Alpha=NamedRerouteUsage())) {sceneTextureId=PPI_PostProcessInput1}
  NamedRerouteDeclaration(None=ComponentMask(None=@n10.Color) {a=true}) {name=ST_Opacity, nodeColor=<rgba 1 0 0.3984 1>}
  NamedRerouteDeclaration(None=ComponentMask(None=@n10.Color) {b=true, g=true, r=true}) {name=Original Scene, nodeColor=<rgba 1 0 0.4165 1>}
  n11 = NamedRerouteUsage()
  n12 = ScalarParameter() {defaultValue=0.7, parameterName=DesatSceneColor}
  NamedRerouteDeclaration(None=Multiply(A=VectorParameter() {parameterName=Background Tint, defaultValue=<rgba 0.0245 0.05 0.016 1>}.RGB, B=Desaturation(None=@n11, Fraction=@n12))) {name=Cold Stuff, nodeColor=<rgba 0 0.4934 1 1>}
  NamedRerouteUsage()
  n13 = LinearInterpolate(A=NamedRerouteUsage(), B=NamedRerouteUsage(), Alpha=NamedRerouteUsage())
  n14 = NamedRerouteUsage()
  n15 = NamedRerouteDeclaration(None=NamedRerouteUsage()) {name=Thermal Combined, nodeColor=<rgba 1 0.3047 0 1>}
  n16 = NamedRerouteUsage()
  StaticSwitchParameter(
   True=LinearInterpolate(A=NamedRerouteUsage(), B=Multiply(A=@n13, B=@n14), Alpha=NamedRerouteUsage())
   False=LinearInterpolate(A=NamedRerouteUsage(), B=Multiply(A=@n15, B=@n16), Alpha=NamedRerouteUsage()) {constAlpha=0}
  ) {group=Active Screen, parameterName=UseActiveScreen, sortPriority=5}
 ]
}
